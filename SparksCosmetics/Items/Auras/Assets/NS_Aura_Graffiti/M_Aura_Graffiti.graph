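# Unlit translucent Niagara particle material for the Graffiti aura, drawing icons from a SubUV atlas.
# Dynamic parameters pick, scale and rotate the frame, whose mask blends drop-shadow color into particle color with a rotatable, noise-broken gradient.
# Opacity multiplies particle alpha, icon-plus-shadow mask, a smoothstepped reveal channel and cloud-noise erosion, both animated per particle.
# The edge-length parameter names are swapped: "Erode Edge Length" softens reveal, "Reveal Edge Length" softens erosion.
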Material /SparksCosmetics/Items/Auras/Assets/NS_Aura_Graffiti/M_Aura_Graffiti
graph {
 properties {
  bAutomaticallySetUsageInEditor=false
  bUsedWithNiagaraMeshParticles=true
  bUsedWithNiagaraRibbons=true
  bUsedWithNiagaraSprites=true
  blendMode=BLEND_Translucent
  opacityMaskClipValue=0.33
  shadingModel=MSM_Unlit
  usedShadingModels=Unlit
 }
 nodes [
  n1 = VectorParameter() {parameterName=Dropshadow Color, defaultValue=<rgba 0 0 0 1>}
  n2 = ParticleColor()
  n3 = Multiply(
   A=TextureCoordinate()
   B=AppendVector(
    A=ScalarParameter() {defaultValue=1, parameterName=Gradient Noise - U}
    B=ScalarParameter() {defaultValue=1, parameterName=Gradient Noise - V}
   )
  )
  n4 = TextureCoordinate()
  n5 = DynamicParameter() {parameterIndex=1, defaultValue=<rgba 0 1 0 0>, paramNames [SubUV Frame, Frame Scale, Gradient Angle, Frame Rotation]}
  n6 = Add(A=ComponentMask(None=MaterialFunctionCall(UVs=@n4, "Rotation Angle (0-1)"=@n5.Gradient Angle) {materialFunction=&CustomRotator}.Rotated Values) {r=true}) {constB=0.3}
  n7 = Add(
   A=Multiply(
    A=TextureSample(UVs=@n3) {samplerType=SAMPLERTYPE_LinearGrayscale, texture=/Game/Effects/Fort_Effects/Textures/Tile/T_WhiteNoise_02_M}.G
    B=@n6
   )
   B=@n6
  )
  n8 = Multiply(A=Saturate(None=Multiply(A=@n7, B=@n6)), B=@n2.RGB)
  n9 = ScalarParameter() {defaultValue=1, parameterName=Gradient On/Off}
  n10 = DynamicParameter() {parameterIndex=1, defaultValue=<rgba 0 1 0 0>, paramNames [SubUV Frame, Frame Scale, Gradient Angle, Frame Rotation]}
  n11 = MaterialFunctionCall(
   UVs=MaterialFunctionCall(UVs=TextureCoordinate(), Rotation Center=Constant2Vector() {g=0.5, r=0.5}, "Rotation Angle (0-1)"=@n10.Frame Rotation) {materialFunction=&CustomRotator}.Rotated Values
   Texture Scale=AppendVector(A=@n10.Frame Scale, B=@n10.Frame Scale)
  ) {materialFunction=&ScaleUVsByCenter}
  n12 = ScalarParameter() {defaultValue=4, parameterName=Mask - SubImage U}
  n13 = ScalarParameter() {defaultValue=2, parameterName=Mask - SubImage V}
  n14 = TextureSampleParameter2D(UVs=MaterialFunctionCall!(UVs=@n11.UVs, SubImages=AppendVector(A=@n12, B=@n13), Frame=@n10.SubUV Frame) {materialFunction=/BRCosmetics/Animation/Game/MainPlayer/Emotes/TroutWrist/FX/Materials/SubUV_Function}.UVs) {parameterName=Mask, texture=&/T_Aura_Graffiti_Icon}
  n15 = ChannelMaskParameter(None=@n14.RGBA) {parameterName=Mask}
  MaterialFunctionCall(Emissive=LinearInterpolate(A=Multiply(A=@n1.RGB, B=@n2.RGB), B=LinearInterpolate(A=@n2.RGB, B=@n8, Alpha=@n9), Alpha=Multiply(A=@n11.0-1 mask, B=@n15.R))) {materialFunction=&MF_FX_EmissiveAdjustments}
  n16 = DynamicParameter() {parameterIndex=1, defaultValue=<rgba 0 1 0 0>, paramNames [SubUV Frame, Frame Scale, Gradient Angle, Frame Rotation]}
  n17 = MaterialFunctionCall(UVs=TextureCoordinate(), Texture Scale=AppendVector(A=@n16.Frame Scale, B=@n16.Frame Scale)) {materialFunction=&ScaleUVsByCenter}
  MaterialFunctionCall(
   UVs=@n17.UVs
   SubImages=AppendVector(
    A=ScalarParameter() {defaultValue=1, parameterName=Reveal - SubImage U}
    B=ScalarParameter() {defaultValue=1, parameterName=Reveal - SubImage V}
   )
   Frame=@n16.SubUV Frame
  ) {materialFunction=&SubUV_Function}
  n18 = ChannelMaskParameter(None=@n14.RGBA) {maskChannel=Green, parameterName=Dropshadow Mask, defaultValue=<rgba 0 1 0 0>}
  n19 = ScalarParameter() {defaultValue=1, parameterName=Dropshadow On/Off}
  n20 = Saturate(None=Add(A=LinearInterpolate(B=@n18.R, Alpha=@n19) {constAlpha=1}, B=@n15.R))
  n21 = DynamicParameter() {defaultValue=<rgba 0 0 1 1>, paramNames [Reveal,Erode,Param3,Param4]}
  n22 = ScalarParameter() {defaultValue=-0.1, parameterName=Erode Edge Length}
  n23 = DynamicParameter() {parameterIndex=1, defaultValue=<rgba 0 1 0 0>, paramNames [SubUV Frame, Frame Scale, Gradient Angle, Frame Rotation]}
  n24 = MaterialFunctionCall(
   UVs=MaterialFunctionCall(UVs=TextureCoordinate(), Rotation Center=Constant2Vector() {g=0.5, r=0.5}, "Rotation Angle (0-1)"=@n23.Frame Rotation) {materialFunction=&CustomRotator}.Rotated Values
   Texture Scale=AppendVector(A=@n23.Frame Scale, B=@n23.Frame Scale)
  ) {materialFunction=&ScaleUVsByCenter}
  n25 = AppendVector(A=ScalarParameter() {defaultValue=4, parameterName=Mask - SubImage U}, B=ScalarParameter() {defaultValue=2, parameterName=Mask - SubImage V})
  n26 = ChannelMaskParameter(None=TextureSampleParameter2D(UVs=MaterialFunctionCall(UVs=@n24.UVs, SubImages=@n25, Frame=@n23.SubUV Frame) {materialFunction=&SubUV_Function}.UVs) {parameterName=Mask, texture=&T_Aura_Graffiti_Icon}.RGBA) {maskChannel=Green, parameterName=Reveal Mask, defaultValue=<rgba 0 1 0 0>}
  n27 = Saturate(None=SmoothStep(Min=Add(A=@n21.Reveal, B=@n22) {constB=-0.1}, Max=@n21.Reveal, Value=Multiply(A=@n17.0-1 mask, B=@n26.R)) {constMax=0, constMin=1})
  n28 = ScalarParameter() {defaultValue=-0.1, parameterName=Reveal Edge Length}
  n29 = Multiply(
   A=TextureCoordinate()
   B=AppendVector(A=ScalarParameter() {defaultValue=1, parameterName=Erode Scale U}, B=ScalarParameter() {defaultValue=1, parameterName=Erode Scale V})
  )
  Multiply(
   A=Multiply(A=@n2.A, B=Multiply(A=@n20, B=@n27))
   B=SmoothStep(
    Min=Add(A=@n21.Erode, B=@n28) {constB=-0.1}
    Max=@n21.Erode
    Value=TextureSample(UVs=@n29) {texture=&AtmosphericCloudNoise01}.G
   ) {constMax=0, constMin=1}
  )
 ]
}
