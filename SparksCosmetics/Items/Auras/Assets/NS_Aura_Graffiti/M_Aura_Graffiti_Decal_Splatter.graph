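# Translucent unlit Niagara sprite material for the Graffiti aura's splatters, tinted by particle color.
# Opacity multiplies particle alpha, a SubUV mask frame, noise-based erosion and a spherical falloff, with frame and erosion driven by dynamic parameters.
# It imitates a decal using inverted depth fade near scene geometry and a Fresnel on flattened scene world normals.
# Low quality and mobile replace that Fresnel with a Forward parameter.
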Material /SparksCosmetics/Items/Auras/Assets/NS_Aura_Graffiti/M_Aura_Graffiti_Decal_Splatter
graph {
 properties {bUsedWithNiagaraSprites=true, blendMode=BLEND_Translucent, pixelDepthOffsetMode=PDOM_Legacy, shadingModel=MSM_Unlit, usedShadingModels=Unlit}
 nodes [
  n1 = ParticleColor()
  MaterialFunctionCall(Emissive=@n1.RGB) {materialFunction=&MF_FX_EmissiveAdjustments}
  n2 = DynamicParameter() {defaultValue=<rgba 3 0.15 1 1>, paramNames [FresnelExponent, FlattenNormal, Param3, Param4]}
  n3 = SceneTexture() {sceneTextureId=PPI_WorldNormal}
  n4 = Fresnel(
   ExponentIn=@n2.FresnelExponent
   Normal=MaterialFunctionCall(Normal=ComponentMask(None=@n3.Color) {b=true, g=true, r=true}, Flatness=@n2.FlattenNormal) {materialFunction=&FlattenNormal}.Result
  ) {baseReflectFraction=0, exponent=1}
  n5 = VectorParameter() {parameterName=Forward, defaultValue=<rgba 0 0 1 1>}
  n6 = Multiply(
   A=ShadingPathSwitch(Default=QualitySwitch(Default=@n4, Low=@n5.RGB, High=@n4, Medium=@n4, Epic=@n4), Mobile=@n5.RGB)
   B=SphereMask(
    A=TextureCoordinate()
    B=Constant() {r=0.5}
    Radius=ScalarParameter() {defaultValue=0.5, parameterName=SphereMaskRadius}
    Hardness=ScalarParameter() {parameterName=SphereMaskHardness}
   ) {hardnessPercent=0}
  )
  n7 = ScalarParameter() {defaultValue=15, parameterName=VisibleDepthDistance}
  n8 = DepthFade(
   Opacity=OneMinus(None=DepthFade(FadeDistance=@n7) {fadeDistanceDefault=80})
   FadeDistance=ScalarParameter() {defaultValue=15, parameterName=DepthFade}
  ) {fadeDistanceDefault=15}
  n9 = TextureCoordinate()
  n10 = AppendVector(A=ScalarParameter() {defaultValue=3, parameterName=SubUV U}, B=ScalarParameter() {defaultValue=2, parameterName=SubUV V})
  n11 = DynamicParameter() {parameterIndex=1, defaultValue=<rgba 0 0 1 1>, paramNames [Erode, SubUV Frame, Param3, Param4]}
  n12 = ChannelMaskParameter(None=TextureSampleParameter2D(UVs=MaterialFunctionCall!(UVs=@n9, SubImages=@n10, Frame=@n11.SubUV Frame) {materialFunction=/BRCosmetics/Animation/Game/MainPlayer/Emotes/TroutWrist/FX/Materials/SubUV_Function}.UVs) {parameterName=Mask, texture=&/T_Aura_Graffiti_Icon}.RGBA) {maskChannel=Alpha, parameterName=Mask Channel, defaultValue=<rgba 0 0 0 1>}
  n13 = ScalarParameter() {defaultValue=-0.1, parameterName=Erode Edge Length}
  n14 = TextureCoordinate() {uTiling=4, vTiling=4}
  n15 = Saturate(None=SmoothStep(
    Min=Add(A=@n11.Erode, B=@n13) {constB=-0.1}
    Max=@n11.Erode
    Value=TextureSample(UVs=@n14) {texture=&AtmosphericCloudNoise01}.G
   ) {constMax=0, constMin=1})
  Multiply(A=@n1.A, B=Multiply(A=Multiply(A=@n6, B=@n8), B=Multiply(A=@n12.R, B=@n15)))
  ShadingPathSwitch(Default=QualitySwitch(Low=Constant()), Mobile=Constant())
 ]
}
