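# Graffiti cosmetic aura Niagara system that layers colorful paint splatters, icons and glows on a character's skeletal mesh.
# Glow and decal splatter sprites track several bones or sockets, while icons and falling splatters over black backing splats share one other.
# Partition-indexed arrays vary color, size and placement, and curve-driven dynamic material parameters animate the sprites.
# User flags adjust for large bodies and dim daytime front-end brightness through time-of-day scaling.
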NiagaraSystem /SparksCosmetics/Items/Auras/Assets/NS_Aura_Graffiti/NS_Aura_Graffiti
dependencies {
 usedDataInterfaces [
  &NiagaraDataInterfaceArrayFloat
  &NiagaraDataInterfaceArrayInt32
  &NiagaraDataInterfaceCurve
  &NiagaraDataInterfaceArrayColor
  &NiagaraDataInterfaceArrayFloat2
  &NiagaraDataInterfaceVector2DCurve
  &NiagaraDataInterfaceSkeletalMesh
 ]
 usedDynamicInputs [&MakeFloatFromVector2d]
 usedModules [
  &self:SystemSpawnScript.SystemScriptSource.SystemScriptGraph.NiagaraNodeAssignment_0.SetVariables_#1
  &SystemState
 ]
}
emitters {
 Decal_Glow {
  emitterUpdateScript [
   &EmitterState()
   &SpawnBurst_Instantaneous(Spawn Count=7, Use Spawn Probability=false, UseLoopCountLimit=true)
  ]
  particleSpawnScript [
   &PartitionParticles(Number Of Partitions=7)
   &InitializeParticle(
    Lifetime=0.1
    Sprite Rotation Mode=::NewEnumerator0
    Sprite Size Mode=::NewEnumerator3
    Uniform Sprite Size=&SelectFloatFromArray(
     Direct Array Index=$Particles.PartitionParticles.Partition
     Float Selection Array=DataInterface() {propertyValues {FloatData [40,40,40,20,20,30,30]}}
    )
    UsePositionOffset=false
    Write Material Random=true
   ) {module=InitializeParticle001}
   &CameraOffset(Camera Offset Amount=10)
  ]
  particleUpdateScript [
   &ParticleState(
    Kill Particles When Lifetime Has Elapsed=false
    Let Infinitely Lived Particles Die When Emitter Deactivates=true
    Loop Particles Lifetime=true
   )
   &SkeletalMeshLocation(
    Bone / Socket Index=&SelectIntFromArray(Direct Array Index=$Particles.PartitionParticles.Partition, Int Selection Array=DataInterface() {propertyValues {IntData [0,1,2,3,4,5,6]}})
    Bone Sampling Mode=::NewEnumerator4
    MaskSpawnGroup=false
    Skeletal Mesh=$System.Skeletal Mesh
   ) {module=SkeletalMeshLocation001}
   &ScaleColor(
    Scale RGB=&VectorFromFloat(Value=&FloatFromCurve(FloatCurve=DataInterface() {propertyValues {Curve {keys [
          {interpMode=RCIM_Cubic, tangentMode=RCTM_User, time=0, value=1}
          {interpMode=RCIM_Cubic, tangentMode=RCTM_User, time=0.1, value=1}
          {interpMode=RCIM_Cubic, tangentMode=RCTM_User, time=1, value=1}
         ]}}}))
    ScaleA=true
    ScaleRGB=true
   )
   &DynamicMaterialParameters(
    Index 0 Param 1=&MakeCustomFloatFromBool(Bool=$User.IsFrontEnd, True Float=5, False Float=5)
    Index 0 Param 2=0.15
    Module.Index0Param0Write=true
    Module.Index0Param1Write=true
    Module.Index0Param2Write=true
    Module.Index0Param3Write=true
   )
  ]
  rendererClasses [&NiagaraSpriteRendererProperties]
  renderers [{rendererClass=&NiagaraSpriteRendererProperties}]
 }
 Decal_Splatter {
  emitterUpdateScript [
   &EmitterState()
   &SpawnRate(SpawnRate=8, Use Spawn Probability=false)
  ]
  particleSpawnScript [
   &PartitionParticles(Number Of Partitions=6)
   &InitializeParticle(
    Color=&SelectLinearColorFromArray(
     Array Sampling Mode=::NewEnumerator0
     Direct Array Index=?
     Color Selection Array=DataInterface() {propertyValues {ColorData [
        <rgba 0.8021 0.1699 0 25>
        <rgba 0.0408 1 0.1108 25>
        <rgba 0.8802 0.3998 0 25>
        <rgba 0 0.743 1 25>
        <rgba 0.1714 0 1 25>
        <rgba 0.7448 0 0.2042 25>
       ]}}
     Random Seed=0
    )
    Lifetime=1.25
    Sprite Rotation Mode=::NewEnumerator2
    Sprite Size=&Multiply_Vector2D_ByFloat(
     Float=&MakeCustomFloatFromBool(Bool=$User.IsLargeBody, True Float=1.5, False Float=1.25)
     Vector2D=&Multiply_Vector2D(
      A=&SelectVector2DFromArray(
       Direct Array Index=$Particles.PartitionParticles.Partition
       Vector 2D Selection Array=DataInterface() {propertyValues {FloatData [<40 40>, <40 40>, <35 35>, <35 35>, <30 30>, <30 30>]}}
      )
      B=<0.6667 1>
     )
    )
    Sprite Size Mode=::NewEnumerator4
    UsePositionOffset=false
    Write Material Random=true
   ) {module=InitializeParticle001}
   &CameraOffset(Camera Offset Amount=&MakeCustomFloatFromBool(Bool=$User.IsLargeBody, True Float=20, False Float=10))
  ]
  particleUpdateScript [
   &ParticleState()
   &SkeletalMeshLocation(
    Bone / Socket Index=&SelectIntFromArray(
     Direct Array Index=$Particles.PartitionParticles.Partition
     Direct Array Mode=::NewEnumerator1
     Int Selection Array=DataInterface() {propertyValues {IntData [1,2,3,4,5,6]}}
    )
    Bone Sampling Mode=::NewEnumerator4
    MaskSpawnGroup=false
    Skeletal Mesh=$System.Skeletal Mesh
   ) {module=SkeletalMeshLocation001}
   &DynamicMaterialParameters(
    Index 0 Param 1=0.5
    Index 1 Param 1=&FloatFromCurve(FloatCurve=DataInterface() {propertyValues {Curve {keys [<at 0 1 RCIM_Cubic>,<at 0.2 0 RCIM_Cubic>,<at 1 1 RCIM_Cubic>]}}})
    Index 1 Param 2=&SelectFloatFromArray(
     Array Sampling Mode=::NewEnumerator0
     Direct Array Index=?
     Float Selection Array=DataInterface() {propertyValues {FloatData [1,4,5]}}
     Random Seed=0
    )
    Module.Index0Param0Write=true
    Module.Index0Param1Write=true
    Module.Index0Param2Write=true
    Module.Index0Param3Write=true
    Module.Index1Param0Write=true
    Module.Index1Param1Write=true
    Module.Index1Param2Write=true
    Module.Index1Param3Write=true
    Write Parameter Index 1=true
   )
   &ScaleColor(
    Scale RGB=&Multiply_VectorByFloat(
     Vector=<1 1 1>
     Float=&FNTimeOfDayScalar(
      Night Scale=1
      Morning Scale=1
      Day Scale=&MakeCustomFloatFromBool(Bool=$User.IsFrontEnd, True Float=0.8, False Float=1)
      Evening Scale=1
     )
    )
    ScaleA=true
    ScaleRGB=true
   )
  ]
  rendererClasses [&NiagaraSpriteRendererProperties]
  renderers [{rendererClass=&NiagaraSpriteRendererProperties}]
 }
 Icons {
  emitterUpdateScript [
   &EmitterState(Life Cycle Mode=::NewEnumerator1, Loop Duration=0.25, UseLoopDelay=false)
   &SpawnBurst_Instantaneous(Use Spawn Probability=false, UseLoopCountLimit=false)
  ]
  particleSpawnScript [
   &PartitionParticles(Number Of Partitions=4, Particles Per Partition=1, Partition Mode:ENiagara_PartitionMode=::NewEnumerator1)
   &InitializeParticle(
    Color=&SelectLinearColorFromArray(
     Array Sampling Mode=::NewEnumerator1
     Direct Array Index=&ReturnParticleIDAttributes()
     Direct Array Mode=::NewEnumerator1
     Color Selection Array=DataInterface() {propertyValues {ColorData [
        <rgba 1 0.0417 0.0671 1>
        <rgba 0.0408 1 0.1108 1>
        <rgba 1 0.4543 0 1>
        <rgba 0 0.2044 0.8177 1>
        <rgba 0.0408 1 0.1108 1>
        <rgba 1 0.4543 0 1>
        <rgba 0.0625 0.2355 1 1>
        <rgba 1 0.0417 0.0671 1>
        <rgba 1 0.4543 0 1>
        <rgba 0.0625 0.2355 1 1>
        <rgba 1 0.0417 0.0671 1>
        <rgba 0.0408 1 0.1108 1>
        <rgba 0.0625 0.2355 1 1>
        <rgba 1 0.0417 0.0671 1>
        <rgba 0.0408 1 0.1108 1>
        <rgba 1 0.4543 0 1>
       ]}}
     Random Seed=?
    )
    Lifetime=&RandomRangeFloat(Minimum=1.25, Maximum=1.25)
    Sprite Size=&Multiply_Vector2D_ByFloat(
     Float=&MakeCustomFloatFromBool(Bool=$User.IsLargeBody, True Float=1.5, False Float=1.3)
     Vector2D=&Multiply_Vector2D(A=<35 35>, B=<0.6667 1>)
    )
    Sprite Size Mode=::NewEnumerator4
    UsePositionOffset=true
    Write Material Random=true
   )
   &SkeletalMeshLocation(Bone / Socket Index=7, Bone Sampling Mode=::NewEnumerator4, MaskSpawnGroup=false, Skeletal Mesh=$System.Skeletal Mesh) {module=SkeletalMeshLocation001}
   /SetVariables(Particles.PivotOffset:Vector2f=&Multiply_Vector2D_ByFloat(
     Float=&MakeCustomFloatFromBool(Bool=$User.IsLargeBody, True Float=1.15, False Float=1)
     Vector2D=&SelectVector2DFromArray(
      Direct Array Index=$Particles.PartitionParticles.Partition
      Direct Array Mode=::NewEnumerator1
      Vector 2D Selection Array=DataInterface() {propertyValues {FloatData [<2 0.35>, <-0.9 0.1>, <1.3 1.2>, <-0.6 1>]}}
     )
    )) {isSetParameters=true}
   &CameraOffset(Camera Offset Amount=-40)
   &DynamicMaterialParameters(
    Index 1 Param 1=&SelectFloatFromArray(
     Direct Array Index=&ReturnParticleIDAttributes()
     Direct Array Mode=::NewEnumerator1
     Float Selection Array=DataInterface() {propertyValues {FloatData [0,1,2,3,4,5,6,7,1,2,3,4,5,6,7,0,2,3,4,5,6,7,0,1,3,4,5,6,7,0,1,2]}}
    )
    Index 1 Param 2=&FloatFromCurve(FloatCurve=DataInterface() {propertyValues {Curve {keys [<at 0 1 RCIM_Cubic>, {arriveTangent=-2.976, interpMode=RCIM_Cubic, leaveTangent=-2.976, tangentMode=RCTM_User, time=1, value=0}]}}})
    Index 1 Param 3=0.125
    Module.Index1Param0Write=true
    Module.Index1Param1Write=false
    Module.Index1Param2Write=true
    Module.Index1Param3Write=true
    Write Parameter Index 0=false
    Write Parameter Index 1=true
   ) {module=DynamicMaterialParameters001}
   &ApplyInitialForces()
  ]
  particleUpdateScript [
   &ParticleState()
   &DynamicMaterialParameters(
    Index 0 Param 1=&FloatFromCurve(FloatCurve=DataInterface() {propertyValues {Curve {keys [<at 0 1 RCIM_Cubic>,<at 0.6 0 RCIM_Cubic>]}}})
    Index 0 Param 2=&FloatFromCurve(FloatCurve=DataInterface() {propertyValues {Curve {keys [
         {arriveTangent=-0.0043, interpMode=RCIM_Cubic, leaveTangent=-0.0043, tangentMode=RCTM_User, time=0.5, value=0}
         {arriveTangent=5.758, interpMode=RCIM_Cubic, leaveTangent=5.758, tangentMode=RCTM_User, time=1, value=1}
        ]}}})
    Index 1 Param 2=&FloatFromCurve(FloatCurve=DataInterface() {propertyValues {Curve {keys [{arriveTangent=0.6979, interpMode=RCIM_Cubic, leaveTangent=0.6979, tangentMode=RCTM_User, time=0, value=0.7}, <at 1 1 RCIM_Cubic>]}}})
    Index 1 Param 3=0.125
    Index 1 Param 4=&FloatFromCurve(
     FloatCurve=DataInterface() {propertyValues {Curve {keys [
         {arriveTangent=0.1746, interpMode=RCIM_Cubic, leaveTangent=0.1746, tangentMode=RCTM_User, time=0, value=0}
         {arriveTangent=2.7048, interpMode=RCIM_Cubic, leaveTangent=2.7048, tangentMode=RCTM_User, time=1, value=1}
        ]}}}
     Scale Curve=&Multiply_Float(
      A=&SelectFloatFromArray(
       Array Sampling Mode=::NewEnumerator0
       Direct Array Index=?
       Float Selection Array=DataInterface() {propertyValues {FloatData [-1,1]}}
       Random Seed=0
      )
      B=0.025
     )
    )
    Module.Index0Param0Write=true
    Module.Index0Param1Write=true
    Module.Index0Param2Write=true
    Module.Index0Param3Write=true
    Module.Index1Param0Write=false
    Module.Index1Param1Write=true
    Module.Index1Param2Write=true
    Module.Index1Param3Write=true
    Write Parameter Index 1=true
   )
   &SkeletalMeshLocation(Bone / Socket Index=7, Bone Sampling Mode=::NewEnumerator4, MaskSpawnGroup=false, Skeletal Mesh=$System.Skeletal Mesh) {module=SkeletalMeshLocation002}
   /SetVariables(Particles.PivotOffset:Vector2f=&Multiply_Vector2D(
     A=&Multiply_Vector2D(
      A=$Particles.Initial.PivotOffset
      B=&Vector2DFromCurve(
       Vector2Curve=DataInterface() {propertyValues {
         XCurve {keys [
           {arriveTangent=1.2878, interpMode=RCIM_Cubic, leaveTangent=1.2878, tangentMode=RCTM_User, time=0, value=0.7}
           {arriveTangent=0.0176, interpMode=RCIM_Cubic, leaveTangent=0.0176, tangentMode=RCTM_User, time=1, value=1.1}
          ]}
         YCurve {keys [
           {arriveTangent=1.2878, interpMode=RCIM_Cubic, leaveTangent=1.2878, tangentMode=RCTM_User, time=0, value=0.7}
           {arriveTangent=0.0176, interpMode=RCIM_Cubic, leaveTangent=0.0176, tangentMode=RCTM_User, time=1, value=1.1}
          ]}
        }}
       ScaleCurve=<1 1>
      )
     )
     B=&MakeVector2D(
      X=&RandomRangeFloat(Minimum=0.75, Maximum=0.9)
      Y=&RandomRangeFloat(Minimum=0.75, Maximum=0.9)
     )
    )) {isSetParameters=true}
   &ScaleColor(
    Scale RGB=&Multiply_VectorByFloat(
     Vector=<1 1 1>
     Float=&FNTimeOfDayScalar(
      Night Scale=1.1
      Morning Scale=1
      Day Scale=&MakeCustomFloatFromBool(Bool=$User.IsFrontEnd, True Float=0.9, False Float=1)
      Evening Scale=1.1
     )
    )
    ScaleA=true
    ScaleRGB=true
   )
  ]
  rendererClasses [&NiagaraSpriteRendererProperties]
  renderers [{rendererClass=&NiagaraSpriteRendererProperties}]
 }
 Splatter {
  emitterUpdateScript [
   &EmitterState()
   &SpawnRate(SpawnRate=8, Use Spawn Probability=false)
  ]
  particleSpawnScript [
   &PartitionParticles(Number Of Partitions=6, Particles Per Partition=1, Partition Mode:ENiagara_PartitionMode=::NewEnumerator1)
   &InitializeParticle(
    Color=&SelectLinearColorFromArray(
     Array Sampling Mode=::NewEnumerator1
     Direct Array Index=$Particles.PartitionParticles.Partition
     Color Selection Array=DataInterface() {propertyValues {ColorData [<rgba 0.7448 0 0.2042 1>, <rgba 1 0.4543 0 1>, <rgba 0 0.743 1 1>, <rgba 0.0408 1 0.1108 1>, <rgba 0.1714 0 1 1>]}}
     Random Seed=?
    )
    Lifetime=0.8
    Sprite Rotation Angle=&SelectFloatFromArray(
     Direct Array Index=$Particles.PartitionParticles.Partition
     Float Selection Array=DataInterface() {propertyValues {FloatData [60,-85,20,40,-50]}}
    )
    Sprite Rotation Mode=::NewEnumerator0
    Sprite Size=&Multiply_Vector2D_ByFloat(
     Float=&MakeCustomFloatFromBool(Bool=$User.IsLargeBody, True Float=1.6, False Float=1.2)
     Vector2D=&Multiply_Vector2D_ByFloat(
      Float=&SelectFloatFromArray(
       Direct Array Index=$Particles.PartitionParticles.Partition
       Float Selection Array=DataInterface() {propertyValues {FloatData [165,80,100,130,110]}}
      )
      Vector2D=<0.66 1>
     )
    )
    Sprite Size Mode=::NewEnumerator4
    UsePositionOffset=true
    Write Material Random=true
   )
   &SkeletalMeshLocation(
    Bone / Socket Index=7
    Bone Sampling Mode=::NewEnumerator4
    MaskSpawnGroup=false
    Position Sampling=::NewEnumerator1
    Skeletal Mesh=$System.Skeletal Mesh
   )
   /SetVariables(Particles.PivotOffset:Vector2f=&SelectVector2DFromArray(
     Direct Array Index=$Particles.PartitionParticles.Partition
     Vector 2D Selection Array=DataInterface() {propertyValues {FloatData [<0.5 0.5>, <0.3 1>, <0.2 0.8>, <0.5 0.5>, <0.5 0.35>]}}
    )) {isSetParameters=true}
   &DynamicMaterialParameters(
    Index 1 Param 1=&SelectFloatFromArray(Direct Array Index=$Particles.PartitionParticles.Partition, Float Selection Array=DataInterface() {propertyValues {FloatData [0,2,3,4,5]}})
    Index 1 Param 3=&SelectFloatFromArray(
     Direct Array Index=$Particles.PartitionParticles.Partition
     Direct Array Mode=::NewEnumerator1
     Float Selection Array=DataInterface() {propertyValues {FloatData [0.5,0.2,0.2,0.5,0.5]}}
    )
    Module.Index1Param0Write=true
    Module.Index1Param1Write=false
    Module.Index1Param2Write=true
    Module.Index1Param3Write=true
    Write Parameter Index 0=false
    Write Parameter Index 1=true
   ) {module=DynamicMaterialParameters001}
  ]
  particleUpdateScript [
   &ParticleState()
   &ScaleSpriteSize(Uniform Curve Sprite Scale=DataInterface() {propertyValues {Curve {keys [
        {arriveTangent=2.6348, interpMode=RCIM_Cubic, leaveTangent=2.6348, tangentMode=RCTM_User, time=0, value=0}
        {tangentMode=RCTM_User, time=1, value=1}
       ]}}})
   &ScaleColor(
    Scale RGB=&Multiply_VectorByFloat(
     Vector=<1 1 1>
     Float=&FNTimeOfDayScalar(
      Night Scale=1.3
      Morning Scale=1.2
      Day Scale=&MakeCustomFloatFromBool(Bool=$User.IsFrontEnd, True Float=0.8, False Float=1)
      Evening Scale=1.3
     )
    )
    ScaleA=true
    ScaleRGB=true
   )
   &DynamicMaterialParameters(
    Index 0 Param 2=&FloatFromCurve(FloatCurve=DataInterface() {propertyValues {Curve {keys [
         <at 0 1 RCIM_Cubic>
         {arriveTangent=-0.0097, interpMode=RCIM_Cubic, leaveTangent=-0.0097, tangentMode=RCTM_User, time=0.4, value=0}
         <at 1 1 RCIM_Cubic>
        ]}}})
    Index 1 Param 2=1
    Index 1 Param 4=&Multiply_Float(
     A=&SelectFloatFromArray(
      Array Sampling Mode=::NewEnumerator0
      Direct Array Index=?
      Float Selection Array=DataInterface() {propertyValues {FloatData [1,1,1,1,1]}}
      Random Seed=0
     )
     B=&RandomRangeFloat(Minimum=0.95, Maximum=1.05)
    )
    Module.Index0Param0Write=true
    Module.Index0Param1Write=true
    Module.Index0Param2Write=true
    Module.Index0Param3Write=true
    Module.Index1Param0Write=false
    Module.Index1Param1Write=true
    Module.Index1Param2Write=false
    Module.Index1Param3Write=true
    Write Parameter Index 1=true
   )
   &SkeletalMeshLocation(
    Bone / Socket Index=7
    Bone Sampling Mode=::NewEnumerator4
    MaskSpawnGroup=false
    Position Sampling=::NewEnumerator1
    Skeletal Mesh=$System.Skeletal Mesh
   ) {module=SkeletalMeshLocation001}
   &GravityForce(Gravity=&RandomRangeVector(Minimum=<0 0 -40>, Maximum=<0 0 -60>))
   &SolveForcesAndVelocity(Clamp Velocity=false, Limit Acceleration=false)
   &CameraOffset(Camera Offset Amount=&Multiply_Float(
     A=&SelectFloatFromArray(
      Direct Array Index=$Particles.PartitionParticles.Partition
      Float Selection Array=DataInterface() {propertyValues {FloatData [-50,-48,-47,-49,-46]}}
     )
     B=&MakeCustomFloatFromBool(Bool=$User.IsLargeBody, True Float=1.75, False Float=1)
    ))
  ]
  rendererClasses [&NiagaraSpriteRendererProperties]
  renderers [{rendererClass=&NiagaraSpriteRendererProperties}]
 }
 SplatterBackground {
  emitterUpdateScript [
   &EmitterState()
   &SpawnRate(SpawnRate=6, Use Spawn Probability=false)
  ]
  particleSpawnScript [
   &PartitionParticles(Number Of Partitions=6, Particles Per Partition=1, Partition Mode:ENiagara_PartitionMode=::NewEnumerator1)
   &InitializeParticle(
    Color=<rgba 0 0 0 1>
    Lifetime=&RandomRangeFloat(Minimum=0.8, Maximum=0.8)
    Sprite Rotation Angle=&SelectFloatFromArray(
     Direct Array Index=$Particles.PartitionParticles.Partition
     Float Selection Array=DataInterface() {propertyValues {FloatData [60,50,-85,20,40,-50]}}
    )
    Sprite Rotation Mode=::NewEnumerator0
    Sprite Size=&Multiply_Vector2D_ByFloat(
     Float=&MakeCustomFloatFromBool(Bool=$User.IsLargeBody, True Float=1.6, False Float=1.2)
     Vector2D=&Multiply_Vector2D(A=<170 170>, B=<0.6667 1>)
    )
    Sprite Size Mode=::NewEnumerator4
    UsePositionOffset=true
    Write Material Random=true
   )
   &SkeletalMeshLocation(
    Bone / Socket Index=7
    Bone Sampling Mode=::NewEnumerator4
    MaskSpawnGroup=false
    Position Sampling=::NewEnumerator1
    Skeletal Mesh=$System.Skeletal Mesh
   )
   &DynamicMaterialParameters(
    Index 1 Param 1=&SelectFloatFromArray(
     Array Sampling Mode=::NewEnumerator0
     Direct Array Index=$Particles.PartitionParticles.Partition
     Direct Array Mode=::NewEnumerator1
     Float Selection Array=DataInterface() {propertyValues {FloatData [1,5]}}
     Random Seed=0
    )
    Index 1 Param 3=&SelectFloatFromArray(
     Direct Array Index=$Particles.PartitionParticles.Partition
     Direct Array Mode=::NewEnumerator1
     Float Selection Array=DataInterface() {propertyValues {FloatData [0.5,0.5]}}
    )
    Module.Index1Param0Write=true
    Module.Index1Param1Write=false
    Module.Index1Param2Write=true
    Module.Index1Param3Write=true
    Write Parameter Index 0=false
    Write Parameter Index 1=true
   ) {module=DynamicMaterialParameters001}
  ]
  particleUpdateScript [
   &ParticleState()
   &ScaleSpriteSize(Uniform Curve Sprite Scale=DataInterface() {propertyValues {Curve {keys [
        {arriveTangent=2.6348, interpMode=RCIM_Cubic, leaveTangent=2.6348, tangentMode=RCTM_User, time=0, value=0}
        {tangentMode=RCTM_User, time=1, value=1}
       ]}}})
   &ScaleColor(ScaleA=true, ScaleRGB=true)
   &DynamicMaterialParameters(
    Index 0 Param 2=&FloatFromCurve(FloatCurve=DataInterface() {propertyValues {Curve {keys [
         <at 0 1 RCIM_Cubic>
         {arriveTangent=-0.0097, interpMode=RCIM_Cubic, leaveTangent=-0.0097, tangentMode=RCTM_User, time=0.4, value=0}
         <at 1 1 RCIM_Cubic>
        ]}}})
    Index 1 Param 2=1
    Module.Index0Param0Write=true
    Module.Index0Param1Write=true
    Module.Index0Param2Write=true
    Module.Index0Param3Write=true
    Module.Index1Param0Write=false
    Module.Index1Param1Write=true
    Module.Index1Param2Write=false
    Module.Index1Param3Write=true
    Write Parameter Index 1=true
   )
   &SkeletalMeshLocation(
    Bone / Socket Index=7
    Bone Sampling Mode=::NewEnumerator4
    MaskSpawnGroup=false
    Position Sampling=::NewEnumerator1
    Skeletal Mesh=$System.Skeletal Mesh
   ) {module=SkeletalMeshLocation001}
   &GravityForce(Gravity=&RandomRangeVector(Minimum=<0 0 -40>, Maximum=<0 0 -60>))
   &SolveForcesAndVelocity(Clamp Velocity=false, Limit Acceleration=false)
   &CameraOffset(Camera Offset Amount=&Multiply_Float(A=-54, B=&MakeCustomFloatFromBool(Bool=true, True Float=1.75, False Float=1))) {module=CameraOffset001}
  ]
  rendererClasses [&NiagaraSpriteRendererProperties]
  renderers [{rendererClass=&NiagaraSpriteRendererProperties}]
 }
}
summary {
 systemName=NS_Aura_Graffiti
 emitters [
  {emitterName=Decal_Glow, rendererClasses [&NiagaraSpriteRendererProperties]}
  {emitterName=Decal_Splatter, rendererClasses [&NiagaraSpriteRendererProperties]}
  {emitterName=Icons, rendererClasses [&NiagaraSpriteRendererProperties]}
  {emitterName=SplatterBackground, rendererClasses [&NiagaraSpriteRendererProperties]}
  {emitterName=Splatter, rendererClasses [&NiagaraSpriteRendererProperties]}
 ]
 userVariables [
  User.Interpolate Intensity {type=NiagaraFloat, defaultValue=3, description=""}
  User.IsFrontEnd {type=NiagaraBool, defaultValue=true}
  User.IsLargeBody {type=NiagaraBool}
 ]
}
